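; Looping gameplay cue for the ImmunityCarriage vehicle's activating state that adds a sound indicator to the cue's looping audio.
; On loop start, after the parent runs, it validates the first audio component, casts the target to Valet_ImmunityCarriage_Vehicle and feeds that component's sound to a targeted looping indicator on the vehicle's Reboot Sound Indicator component.
; On removal, it calls the parent, then stops that indicator loop.
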(event EventOnLoopingStart(EitherVFX-preferthisversion)
  (params (MyTarget :type Actor%20Object%20Reference) (Parameters :type Gameplay%20Cue%20Parameters%20Structure%20%28by%20ref%29) (ParticleComponents :type Array%20of%20FXSystem%20Component%20Object%20References) (AudioComponents :type Array%20of%20Audio%20Component%20Object%20References))
  (bind _output (Utilities|Array|Get(acopy) AudioComponents))
  (|CallParentFunction MyTarget Parameters ParticleComponents AudioComponents)
  (if (Utilities|Array|IsValidIndex AudioComponents)
    (Utilities|IsValid _output
      (:Is%20Valid
        (bind _asvalet_immunity_carriage_vehicle (Utilities|Casting|CastToValet_ImmunityCarriage_Vehicle MyTarget))
        (bind _reboot_sound_indicator_component (Variables|/ImmunityCarriage/Vehicle/Valet_ImmunityCarriage_Vehicle.Valet_ImmunityCarriage_Vehicle_C|GetReboot%20Sound%20Indicator%20Component _asvalet_immunity_carriage_vehicle)
          (:then
            (SoundIndicator|StartTargetedSoundIndicatorLooping _reboot_sound_indicator_component _asvalet_immunity_carriage_vehicle (Variables|/Script/Engine.AudioComponent|GetSound _output) 1.0 0))
          (:else))))))

(event EventOnRemoval(EitherVFX-preferthisversion)
  (params (MyTarget :type Actor%20Object%20Reference) (Parameters :type Gameplay%20Cue%20Parameters%20Structure%20%28by%20ref%29) (ParticleComponents :type Array%20of%20FXSystem%20Component%20Object%20References) (AudioComponents :type Array%20of%20Audio%20Component%20Object%20References) (BurstCameraShakeInstance :type Camera%20Shake%20Base%20Object%20Reference) (BurstDecalInstance :type Decal%20Actor%20Object%20Reference))
  (|CallParentFunction MyTarget Parameters ParticleComponents AudioComponents BurstCameraShakeInstance BurstDecalInstance)
  (bind _asvalet_immunity_carriage_vehicle (Utilities|Casting|CastToValet_ImmunityCarriage_Vehicle MyTarget))
  (bind _reboot_sound_indicator_component (Variables|/ImmunityCarriage/Vehicle/Valet_ImmunityCarriage_Vehicle.Valet_ImmunityCarriage_Vehicle_C|GetReboot%20Sound%20Indicator%20Component _asvalet_immunity_carriage_vehicle)
    (:then
      (SoundIndicator|StopSoundIndicatorLooping _reboot_sound_indicator_component))
    (:else)))

(fn ConstructionScript)
