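; Looping gameplay cue for the active Immunity Carriage resurrection van, driving its Niagara effects.
; On loop start it stores the warmup effect, moves it to the vehicle's holo-logo component and binds to the pawn-resurrected event.
; An enhanced resurrection records a flag and, after a short delay, switches the warmup effect to Variant 1.
; On removal it unbinds, sets the removal effect's Variant from that flag and positions it at the holo-logo.
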(event EventOnLoopingStart(EitherVFX-preferthisversion)
  (params (MyTarget :type Actor%20Object%20Reference) (Parameters :type Gameplay%20Cue%20Parameters%20Structure%20%28by%20ref%29) (ParticleComponents :type Array%20of%20FXSystem%20Component%20Object%20References) (AudioComponents :type Array%20of%20Audio%20Component%20Object%20References))
  (bind _output (Utilities|Array|Get(acopy) ParticleComponents))
  ;[0 0.2083 0.03411] OnLoopingStart: Update the position of the spawned nigara effect and bind to know when enhanced resurrection is happening on the van.
  (|CallParentFunction MyTarget Parameters ParticleComponents AudioComponents)
  (bind _asimmunity_carriage_pawn_resurrection_interface (Utilities|Casting|CastToImmunityCarriage_PawnResurrectionInterface MyTarget))
  (bind _asbpi_valet_immunity_carriage_vehicle (Utilities|Casting|CastToBPI_ValetImmunityCarriageVehicle MyTarget))
  (if (Utilities|Array|IsValidIndex ParticleComponents)
    (Utilities|IsValid _output
      (:Is%20Valid
        (Variables|Default|SetSpawnWarmupFXSystemComponent _output)
        (ImmunityCarriage_PawnResurrectionInterface|BindtoonPawnResurrectedEvent _asimmunity_carriage_pawn_resurrection_interface (EventDispatchers|CreateEvent :!event Custom|HandlePawnResurrected) :!class /Script/ImmunityCarriageRuntime.ImmunityCarriage_PawnResurrectionInterface)
        (bind _holologonscomponent (BPI_ValetImmunityCarriageVehicle|GetSCMachineHoloLogoNiagaraComponent _asbpi_valet_immunity_carriage_vehicle :!class /ImmunityCarriage/Vehicle/BPI_ValetImmunityCarriageVehicle.BPI_ValetImmunityCarriageVehicle_C))
        (Transformation|SetWorldLocation _output (Transformation|GetWorldLocation _holologonscomponent) false false)))))

(event EventOnRemoval(EitherVFX-preferthisversion)
  (params (MyTarget :type Actor%20Object%20Reference) (Parameters :type Gameplay%20Cue%20Parameters%20Structure%20%28by%20ref%29) (ParticleComponents :type Array%20of%20FXSystem%20Component%20Object%20References) (AudioComponents :type Array%20of%20Audio%20Component%20Object%20References) (BurstCameraShakeInstance :type Camera%20Shake%20Base%20Object%20Reference) (BurstDecalInstance :type Decal%20Actor%20Object%20Reference))
  (bind _output (Utilities|Array|Get(acopy) ParticleComponents))
  ;[0.15 0.07615 0 0.5] OnRemoval: Unbind delegates.
  (|CallParentFunction MyTarget Parameters ParticleComponents AudioComponents BurstCameraShakeInstance BurstDecalInstance)
  (if (Utilities|Array|IsValidIndex ParticleComponents)
    (bind _asimmunity_carriage_pawn_resurrection_interface (Utilities|Casting|CastToImmunityCarriage_PawnResurrectionInterface MyTarget))
    (bind _asbpi_valet_immunity_carriage_vehicle (Utilities|Casting|CastToBPI_ValetImmunityCarriageVehicle MyTarget))
    (ImmunityCarriage_PawnResurrectionInterface|UnbindfromonPawnResurrectedEvent _asimmunity_carriage_pawn_resurrection_interface (EventDispatchers|CreateEvent :!event Custom|HandlePawnResurrected) :!class /Script/ImmunityCarriageRuntime.ImmunityCarriage_PawnResurrectionInterface)
    (bind _asniagara_particle_system_component (Utilities|Casting|CastToNiagaraComponent _output))
    (if (Variables|Default|GetEnhanced%20Resurrection)
      (Niagara|SetNiagaraVariable(Int32) _asniagara_particle_system_component "Variant" 1)
      (else
        (Niagara|SetNiagaraVariable(Int32) _asniagara_particle_system_component "Variant" 0)))
    (bind _holologonscomponent (BPI_ValetImmunityCarriageVehicle|GetSCMachineHoloLogoNiagaraComponent _asbpi_valet_immunity_carriage_vehicle :!class /ImmunityCarriage/Vehicle/BPI_ValetImmunityCarriageVehicle.BPI_ValetImmunityCarriageVehicle_C))
    (Transformation|SetWorldLocation _output (Transformation|GetWorldLocation _holologonscomponent) false false)))

(event Custom|HandlePawnResurrected
  (params (FortPawn :type Fort%20Pawn%20Object%20Reference) (bEnhancedResurrection :type Boolean))
  ;[0.15 0 0.1222 0.5] HandlePawnResurrected: When enhanced resurrection is happening, update the warmup fx to variant 1 for a short duration.
  (Variables|Default|SetEnhanced%20Resurrection bEnhancedResurrection)
  (if bEnhancedResurrection
    (Utilities|FlowControl|Delay self 0.5)
    (bind _spawnwarmupfxsystemcomponent (Variables|Default|GetSpawnWarmupFXSystemComponent)
      (:then
        (bind _asniagara_particle_system_component (Utilities|Casting|CastToNiagaraComponent _spawnwarmupfxsystemcomponent))
        (Niagara|SetNiagaraVariable(Int32) _asniagara_particle_system_component "Variant" 1))
      (:else))))

(fn ConstructionScript)
